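# Animation Blueprint for the female medium StarfishUniform_01 player body mesh, following the pose of the mesh it is attached to.
# The copied parent pose runs through the outfit's Control Rig, then a spring-driven, angle-limited Anim Dynamics body on the front bandolier bone.
# Both thigh helper bones are then shrunk by a fixed replace-mode scale.
# The event graph stores the owning mesh in MeshToCopy, which the AnimGraph never reads.
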AnimBlueprint /Game/Characters/Player/Female/Medium/Bodies/F_MED_StarfishUniform_01/Meshes/F_MED_StarfishUniform_01_AnimBP
target_skeleton=&/F_MED_StarfishUniform_01_Skeleton
graphs [
 AnimGraph {nodes [
   n1 = Control Rig(Source=Copy Pose From Mesh() {bUseAttachedParent=true}) {
    alphaBoolBlend {blendOption=Linear}
    controlRigAssetReference {blueprintRigClass=&/F_MED_StarfishUniform_01ControlRig.F_MED_StarfishUniform_01ControlRig_C}
   }
   n2 = Anim Dynamics - Bone: dyn_bandolier_fr_02(ComponentPose=Local To Component(LocalPose=@n1), Alpha=1, ExternalForce="0,0,0") {
    angularDampingOverride=0.99
    angularSpringConstant=10000
    bAngularSpring=true
    bGravityOverrideInSimSpace=true
    bOverrideAngularDamping=true
    numSolverIterationsPostUpdate=4
    numSolverIterationsPreUpdate=16
    alphaBoolBlend {blendOption=Linear}
    boundBone {boneName=dyn_bandolier_fr_02}
    physicsBodyDefinitions [{
      sphereCollisionRadius=0
      boundBone {boneName=dyn_bandolier_fr_02}
      boxExtents=<5 30 30>
      constraintSetup {angularTargetAxis=AxisZ, angularLimitsMax=<2 5 10>, angularLimitsMin=<-2 -20 -10>, angularTarget=<0 0 1>}
      localJointOffset=<-1 0 0>
     }]
   }
   n3 = "Transform (Modify) Bone - Bone: deform_thigh_helper_l"(ComponentPose=@n2, Alpha=1, Translation="0,0,0", Rotation="0,0,0", Scale="0.7,0.7,0.7") {
    scaleMode=BMM_Replace
    scaleSpace=BCS_BoneSpace
    alphaBoolBlend {blendOption=Linear}
    boneToModify {boneName=deform_thigh_helper_l}
    scale=<0.7 0.7 0.7>
   }
   n4 = "Transform (Modify) Bone - Bone: deform_thigh_helper_r"(ComponentPose=@n3, Alpha=1, Translation="0,0,0", Rotation="0,0,0", Scale="0.7,0.7,0.7") {
    scaleMode=BMM_Replace
    scaleSpace=BCS_BoneSpace
    alphaBoolBlend {blendOption=Linear}
    boneToModify {boneName=deform_thigh_helper_r}
    scale=<0.7 0.7 0.7>
   }
   Output Pose(Result=Component To Local(ComponentPose=@n4)) {layerGroup=None}
  ]}
 EventGraph {nodes [
   Event BlueprintInitializeAnimation()
   > Set MeshToCopy(MeshToCopy:Skeletal Mesh Component Object Reference=GetOwnerSkeletalMeshComponent()) -> (Output_Get:Skeletal Mesh Component Object Reference)
   TryGetPawnOwner()
   Event BlueprintUpdateAnimation()
  ]}
]
variables [MeshToCopy {type=SkeletalMeshComponent}]
